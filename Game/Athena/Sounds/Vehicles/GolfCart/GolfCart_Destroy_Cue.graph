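# GolfCart_Destroy_Cue is a Sound Cue for the golf cart vehicle's destruction sound.
# A Random node picks between two destroy wave variations with equal weight.
# It plays in the vehicles sound class with a boosted volume, spatialized through a golf cart distant movement attenuation asset rather than a destruction-specific one.
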SoundCue /Game/Athena/Sounds/Vehicles/GolfCart/GolfCart_Destroy_Cue
root=Random_0
nodes [
 {id=Random_0, children [WavePlayer_1,WavePlayer_2], properties {weights [1,1]}}
 {id=WavePlayer_1, properties {soundWaveAssetPtr=&/GolfCart_Destroy_01}}
 {id=WavePlayer_2, properties {soundWaveAssetPtr=&/GolfCart_Destroy_02}}
]
properties {
 attenuationSettings=/Game/Sounds/Attenuation/Fort_GolfCart_Movement_Distant_Att
 maxDistance=10100
 soundClassObject=&Fort_Vehicles
 volumeMultiplier=1.25
}
