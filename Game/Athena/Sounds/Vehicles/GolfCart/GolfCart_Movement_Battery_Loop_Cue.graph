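# Looping sound cue for the golf cart's battery whirr in the vehicle movement audio, in the Fort_Vehicles sound class.
# A quality-level node picks one of two looping waves: the first two levels use the standard whirr, the last two an OPT_02 variant.
# A close-detail golf cart movement attenuation asset spatializes it, with a slightly reduced cue volume.
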SoundCue /Game/Athena/Sounds/Vehicles/GolfCart/GolfCart_Movement_Battery_Loop_Cue
root=QualityLevel_0
nodes [
 {id=QualityLevel_0, children [WavePlayer_1,WavePlayer_1,WavePlayer_2,WavePlayer_2]}
 {id=WavePlayer_1, properties {bLooping=true, soundWaveAssetPtr=&/GolfCart_BatteryWhirrLoop_34}}
 {id=WavePlayer_2, properties {bLooping=true, soundWaveAssetPtr=&/GolfCart_BatteryWhirrLoop_34_OPT_02}}
]
properties {
 attenuationSettings=/Game/Sounds/Attenuation/ATT_Fort_GolfCart_Movement_Close_Detail
 maxDistance=10000
 soundClassObject=&Fort_Vehicles
 volumeMultiplier=0.8
}
