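# Sound cue for the golf cart vehicle's boost activation sound.
# It picks one of two booster-increase wave variations at random with equal weight, then a modulator randomizes pitch slightly around normal.
# Playback uses the golf cart boost movement attenuation settings and the vehicles sound class.
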SoundCue /Game/Athena/Sounds/Vehicles/GolfCart/GolfCart_Movement_BoostActivate_Cue
root=Modulator_0
nodes [
 {id=Modulator_0, children [Random_0], properties {pitchMax=1.1, pitchMin=0.9}}
 {id=Random_0, children [WavePlayer_0,WavePlayer_1], properties {weights [1,1]}}
 {id=WavePlayer_0, properties {soundWaveAssetPtr=&/GolfCart_BoosterIncrease_32}}
 {id=WavePlayer_1, properties {soundWaveAssetPtr=&/GolfCart_BoosterIncrease_35}}
]
properties {
 attenuationSettings=/Game/Sounds/Attenuation/ATT_Fort_GolfCart_Movement_Boost
 maxDistance=12000
 soundClassObject=&Fort_Vehicles
}
