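# Sound cue for the golf cart's level 1 boost increase movement sound, reusing jetpack thrust samples.
# A Random node picks one of three equally weighted jetpack thrust waves, which feeds a Modulator that varies pitch slightly on each play.
# It plays in the vehicles sound class at reduced volume, using a golf cart boost attenuation asset for distance falloff.
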SoundCue /Game/Athena/Sounds/Vehicles/GolfCart/GolfCart_Movement_BoostLevelIncrease_1_Cue_NEW
root=Modulator_0
nodes [
 {id=Modulator_0, children [Random_1], properties {pitchMax=1.1, pitchMin=0.9}}
 {id=Random_1, children [WavePlayer_3,WavePlayer_4,WavePlayer_5], properties {weights [1,1,1]}}
 {id=WavePlayer_3, properties {soundWaveAssetPtr=/Game/Athena/Sounds/Items/Jetpack/Jetpack_Trust_03}}
 {id=WavePlayer_4, properties {soundWaveAssetPtr=/Game/Athena/Sounds/Items/Jetpack/Jetpack_Trust_01}}
 {id=WavePlayer_5, properties {soundWaveAssetPtr=/Game/Athena/Sounds/Items/Jetpack/Jetpack_Trust_02}}
]
properties {
 attenuationSettings=/Game/Sounds/Attenuation/ATT_Fort_GolfCart_Movement_Boost
 maxDistance=12000
 soundClassObject=&Fort_Vehicles
 volumeMultiplier=0.35
}
