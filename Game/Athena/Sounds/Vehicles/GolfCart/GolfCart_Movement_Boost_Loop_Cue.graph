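# Looping sound cue for the golf cart's movement boost in Athena, in the vehicles sound class with its own attenuation.
# A continuous modulator at the root maps a Pitch parameter to a bounded pitch range, so pitch can vary during playback.
# It feeds a quality-level switch that plays the standard boost loop for the first two levels and the "opt" variant for the last two.
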SoundCue /Game/Athena/Sounds/Vehicles/GolfCart/GolfCart_Movement_Boost_Loop_Cue
root=ModulatorContinuous_0
nodes [
 {id=ModulatorContinuous_0, children [QualityLevel_0], properties {pitchModulationParams {maxOutput=1.2, minOutput=0.75, parameterName=Pitch}}}
 {id=QualityLevel_0, children [WavePlayer_0,WavePlayer_0,WavePlayer_1,WavePlayer_1]}
 {id=WavePlayer_0, properties {bLooping=true, soundWaveAssetPtr=&/GolfCart_BoostLoop}}
 {id=WavePlayer_1, properties {bLooping=true, soundWaveAssetPtr=&/GolfCart_BoostLoop_opt}}
]
properties {
 attenuationSettings=/Game/Sounds/Attenuation/ATT_Fort_GolfCart_Movement_Boost
 maxDistance=12000
 soundClassObject=&Fort_Vehicles
 volumeMultiplier=1
}
