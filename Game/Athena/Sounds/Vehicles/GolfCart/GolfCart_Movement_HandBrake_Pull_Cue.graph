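# Sound cue for the golf cart handbrake pull in Athena vehicle audio.
# A Random node picks between two equally weighted handbrake pull wave variations.
# It plays at reduced volume through the vehicles sound class, spatialized with the golf cart close suspension attenuation asset.
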SoundCue /Game/Athena/Sounds/Vehicles/GolfCart/GolfCart_Movement_HandBrake_Pull_Cue
root=Random_0
nodes [
 {id=Random_0, children [WavePlayer_2,WavePlayer_0], properties {weights [1,1]}}
 {id=WavePlayer_0, properties {soundWaveAssetPtr=&/GolfCart_HandbrakePull_38}}
 {id=WavePlayer_2, properties {soundWaveAssetPtr=&/GolfCart_HandbrakePull_28}}
]
properties {
 attenuationSettings=/Game/Sounds/Attenuation/ATT_Fort_GolfCart_Movement_Close_Suspension
 maxDistance=8500
 soundClassObject=&Fort_Vehicles
 volumeMultiplier=0.5
}
