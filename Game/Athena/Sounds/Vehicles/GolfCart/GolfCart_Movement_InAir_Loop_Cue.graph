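# Sound cue for the golf cart's in-air movement loop, playing the tire sound heard while the vehicle is airborne.
# A quality-level node at the root picks between the standard loop wave for the two higher tiers and an optimized variant for the two lower tiers.
# It routes to the Fortnite vehicles sound class at reduced volume and uses the golf cart movement boost attenuation asset.
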SoundCue /Game/Athena/Sounds/Vehicles/GolfCart/GolfCart_Movement_InAir_Loop_Cue
root=QualityLevel_0
nodes [
 {id=QualityLevel_0, children [WavePlayer_0,WavePlayer_0,WavePlayer_1,WavePlayer_1]}
 {id=WavePlayer_0, properties {soundWaveAssetPtr=&/GolfCart_Tire_InAirLoop_01}}
 {id=WavePlayer_1, properties {soundWaveAssetPtr=&/GolfCart_Tire_InAirLoop_01_opt}}
]
properties {
 attenuationSettings=/Game/Sounds/Attenuation/ATT_Fort_GolfCart_Movement_Boost
 maxDistance=12000
 soundClassObject=&Fort_Vehicles
 volumeMultiplier=0.6
}
