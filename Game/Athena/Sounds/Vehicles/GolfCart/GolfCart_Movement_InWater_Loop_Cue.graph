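# Sound cue for the looping sound of the golf cart vehicle moving through water.
# A single looping wave player plays the golf cart water loop wave as the root, with no mixing or modulation nodes.
# The cue routes to the vehicles sound class, uses a close-range golf cart movement attenuation asset, and boosts its volume.
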SoundCue /Game/Athena/Sounds/Vehicles/GolfCart/GolfCart_Movement_InWater_Loop_Cue
root=WavePlayer_1
nodes [{id=WavePlayer_1, properties {bLooping=true, soundWaveAssetPtr=&/GolfCart_WaterLoop_01}}]
properties {
 attenuationSettings=/Game/Sounds/Attenuation/Fort_GolfCart_Movement_Close_Att
 maxDistance=5300
 soundClassObject=&Fort_Vehicles
 volumeMultiplier=1.5
}
