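# Looping movement sound cue for the Athena golf cart, layering engine and tire loops under a Doppler node.
# The engine branch mixes a Pitch-modulated acceleration loop with a medium movement loop, fading in with the Speed parameter through engine attenuation.
# The quieter tire branch fades in at very low Speed and out toward full Speed, using close-detail attenuation.
# Quality-level nodes swap each loop for an optimized wave on lower settings.
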SoundCue /Game/Athena/Sounds/Vehicles/GolfCart/GolfCart_Movement_Move_Loop_Cue
root=Doppler_0
nodes [
 {id=Attenuation_0, children [ParamCrossFade_2], properties {attenuationSettings=/Game/Sounds/Attenuation/ATT_Fort_GolfCart_Engine}}
 {
  id=Attenuation_2
  children [ParamCrossFade_3]
  properties {attenuationSettings=/Game/Sounds/Attenuation/ATT_Fort_GolfCart_Movement_Close_Detail}
 }
 {id=Doppler_0, children [Mixer_0]}
 {id=Mixer_0, children [Attenuation_0, Attenuation_2], properties {inputVolume [1,0.5]}}
 {id=Mixer_2, children [ModulatorContinuous_0, QualityLevel_2], properties {inputVolume [1.5,1]}}
 {id=ModulatorContinuous_0, children [QualityLevel_0], properties {pitchModulationParams {maxOutput=1.3, minOutput=0.7, parameterName=Pitch}}}
 {
  id=ParamCrossFade_2
  children [Mixer_2]
  properties {paramName=Speed, crossFadeInput [{fadeInDistanceEnd=0.7, fadeInDistanceStart=0, fadeOutDistanceEnd=100, fadeOutDistanceStart=100, volume=1}]}
 }
 {
  id=ParamCrossFade_3
  children [QualityLevel_1]
  properties {paramName=Speed, crossFadeInput [{fadeInDistanceEnd=0.05, fadeInDistanceStart=0, fadeOutDistanceEnd=1, fadeOutDistanceStart=0.35, volume=1}]}
 }
 {id=QualityLevel_0, children [WavePlayer_0,WavePlayer_0,WavePlayer_1,WavePlayer_1]}
 {id=QualityLevel_1, children [WavePlayer_8,WavePlayer_8,WavePlayer_2,WavePlayer_2]}
 {id=QualityLevel_2, children [WavePlayer_7,WavePlayer_7,WavePlayer_3,WavePlayer_3]}
 {id=WavePlayer_0, properties {bLooping=true, soundWaveAssetPtr=&/GolfCart_Engine_AccelerationLoop_02}}
 {id=WavePlayer_1, properties {bLooping=true, soundWaveAssetPtr=&/GolfCart_Engine_AccelerationLoop_02_OPT}}
 {id=WavePlayer_2, properties {bLooping=true, soundWaveAssetPtr=&/GolfCart_Tire_MovementLoop_01_OPT}}
 {
  id=WavePlayer_3
  properties {bLooping=true, soundWaveAssetPtr=&/GolfCart_Engine_MovementLoop_Medium_01_opt}
 }
 {id=WavePlayer_7, properties {bLooping=true, soundWaveAssetPtr=&/GolfCart_Engine_MovementLoop_Medium_01}}
 {id=WavePlayer_8, properties {bLooping=true, soundWaveAssetPtr=&/GolfCart_Tire_MovementLoop_01}}
]
properties {maxDistance=12000, soundClassObject=&Fort_Vehicles, volumeMultiplier=0.9}
