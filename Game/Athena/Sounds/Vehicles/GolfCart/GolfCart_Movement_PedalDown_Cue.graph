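# Sound cue for the golf cart vehicle's pedal-down movement sound.
# A Random node picks one of four equally weighted pedal-down wave variations on each play.
# Output plays at reduced volume in the Fort vehicles sound class, using a close-detail golf cart movement attenuation.
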SoundCue /Game/Athena/Sounds/Vehicles/GolfCart/GolfCart_Movement_PedalDown_Cue
root=Random_2
nodes [
 {id=Random_2, children [WavePlayer_5,WavePlayer_6,WavePlayer_7,WavePlayer_8], properties {weights [1,1,1,1]}}
 {id=WavePlayer_5, properties {soundWaveAssetPtr=&/GolfCart_PedalDown_15}}
 {id=WavePlayer_6, properties {soundWaveAssetPtr=&/GolfCart_PedalDown_16}}
 {id=WavePlayer_7, properties {soundWaveAssetPtr=&/GolfCart_PedalDown_17}}
 {id=WavePlayer_8, properties {soundWaveAssetPtr=&/GolfCart_PedalDown_18}}
]
properties {
 attenuationSettings=/Game/Sounds/Attenuation/ATT_Fort_GolfCart_Movement_Close_Detail
 maxDistance=10000
 soundClassObject=&Fort_Vehicles
 volumeMultiplier=0.5
}
