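# Sound cue that plays the looping scrape sound for the golf cart vehicle's movement.
# A quality level node feeds the same looping scrape wave into its first two quality slots and leaves the last two slots empty.
# It uses the golf cart close collision impact attenuation, routes through the vehicles sound class and plays at reduced volume.
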SoundCue /Game/Athena/Sounds/Vehicles/GolfCart/GolfCart_Movement_Scrape_Cue
root=QualityLevel_0
nodes [
 {id=QualityLevel_0, children [WavePlayer_2,WavePlayer_2,null,null]}
 {id=WavePlayer_2, properties {bLooping=true, soundWaveAssetPtr=&/GolfCart_ScrapeLoop}}
]
properties {
 attenuationSettings=/Game/Sounds/Attenuation/ATT_Fort_GolfCart_Collision_Impact_Close
 maxDistance=10000
 soundClassObject=&Fort_Vehicles
 volumeMultiplier=0.55
}
