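# Sound cue that plays the golf cart vehicle's movement stop sound.
# A Random node picks one of three equally weighted stop wave variations each time it plays.
# It routes to the vehicles sound class at reduced volume and uses a close-detail golf cart movement attenuation for spatial falloff.
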SoundCue /Game/Athena/Sounds/Vehicles/GolfCart/GolfCart_Movement_Stop_Cue
root=Random_0
nodes [
 {id=Random_0, children [WavePlayer_0,WavePlayer_2,WavePlayer_3], properties {weights [1,1,1]}}
 {id=WavePlayer_0, properties {soundWaveAssetPtr=&/GolfCart_Stop_12}}
 {id=WavePlayer_2, properties {soundWaveAssetPtr=&/GolfCart_Stop_13}}
 {id=WavePlayer_3, properties {soundWaveAssetPtr=&/GolfCart_Stop_14}}
]
properties {
 attenuationSettings=/Game/Sounds/Attenuation/ATT_Fort_GolfCart_Movement_Close_Detail
 maxDistance=10000
 soundClassObject=&Fort_Vehicles
 volumeMultiplier=0.6
}
