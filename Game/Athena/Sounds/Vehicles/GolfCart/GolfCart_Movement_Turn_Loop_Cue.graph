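# Sound cue for the looping tire sound of a turning golf cart, using the vehicle sound class and golf cart movement attenuation.
# A Strength parameter crossfades from TurnLoop_01 at zero to TurnLoop_03 at one.
# The TurnLoop_01 branch passes through a quality-level switch that selects either the standard wave or its "opt" variant, while TurnLoop_03 plays the same wave at every quality level.
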SoundCue /Game/Athena/Sounds/Vehicles/GolfCart/GolfCart_Movement_Turn_Loop_Cue
root=ParamCrossFade_0
nodes [
 {
  id=ParamCrossFade_0
  children [QualityLevel_0, WavePlayer_4]
  properties {
   paramName=Strength
   crossFadeInput [
    {fadeInDistanceEnd=0, fadeInDistanceStart=0, fadeOutDistanceEnd=1, fadeOutDistanceStart=0, volume=1}
    {fadeInDistanceEnd=1, fadeInDistanceStart=0, fadeOutDistanceEnd=100, fadeOutDistanceStart=100, volume=1}
   ]
  }
 }
 {id=QualityLevel_0, children [WavePlayer_2,WavePlayer_2,WavePlayer_0,WavePlayer_0]}
 {id=WavePlayer_0, properties {bLooping=true, soundWaveAssetPtr=&/GolfCart_Tire_TurnLoop_01_opt}}
 {id=WavePlayer_2, properties {bLooping=true, soundWaveAssetPtr=&/GolfCart_Tire_TurnLoop_01}}
 {id=WavePlayer_4, properties {bLooping=true, soundWaveAssetPtr=&/GolfCart_Tire_TurnLoop_03}}
]
properties {
 attenuationSettings=/Game/Sounds/Attenuation/ATT_Fort_GolfCart_Movement
 maxDistance=12512
 soundClassObject=&Fort_Vehicles
 volumeMultiplier=0.6
}
