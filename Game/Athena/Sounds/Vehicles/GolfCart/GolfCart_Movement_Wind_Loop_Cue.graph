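# Looping wind sound cue for golf cart movement, routed to the Fort_Vehicles sound class.
# A quality-level root feeds Wind_UmbrellaClosed_Loop_01 to the first two quality slots and the 1P_opt_02 loop to the last two.
# Both wave players loop and reuse closed-umbrella wind assets from the Wind_Falling folder.
# The golf cart close-detail movement attenuation asset sets its spatial falloff.
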SoundCue /Game/Athena/Sounds/Vehicles/GolfCart/GolfCart_Movement_Wind_Loop_Cue
root=QualityLevel_0
nodes [
 {id=QualityLevel_0, children [WavePlayer_4,WavePlayer_4,WavePlayer_0,WavePlayer_0]}
 {id=WavePlayer_0, properties {bLooping=true, soundWaveAssetPtr=/Game/Athena/Sounds/Wind_Falling/Wind_UmbrellaClosed_Loop_1P_opt_02}}
 {id=WavePlayer_4, properties {bLooping=true, soundWaveAssetPtr=/Game/Athena/Sounds/Wind_Falling/Wind_UmbrellaClosed_Loop_01}}
]
properties {
 attenuationSettings=/Game/Sounds/Attenuation/ATT_Fort_GolfCart_Movement_Close_Detail
 maxDistance=10000
 soundClassObject=&Fort_Vehicles
 volumeMultiplier=1
}
