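# Sound cue for bouncing on the golf cart roof, played as player foley in Athena.
# It picks one of three equally weighted LaunchBounce waves at random, then passes it through a modulator that varies playback, with maximum pitch capped at normal.
# The cue boosts volume, uses close-detail golf cart movement attenuation, and overrides concurrency to block new instances once its limit is reached.
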SoundCue /Game/Athena/Sounds/Vehicles/GolfCart/GolfCart_Roof_Bounce_Cue
root=Modulator_1
nodes [
 {id=Modulator_1, children [Random_0], properties {pitchMax=1}}
 {id=Random_0, children [WavePlayer_2,WavePlayer_5,WavePlayer_6], properties {weights [1,1,1]}}
 {id=WavePlayer_2, properties {soundWaveAssetPtr=/Game/Athena/Sounds/Vehicles/GolfCart/PlayerFoley/GoflCart_LaunchBounce_01}}
 {id=WavePlayer_5, properties {soundWaveAssetPtr=/Game/Athena/Sounds/Vehicles/GolfCart/PlayerFoley/GoflCart_LaunchBounce_02}}
 {id=WavePlayer_6, properties {soundWaveAssetPtr=/Game/Athena/Sounds/Vehicles/GolfCart/PlayerFoley/GoflCart_LaunchBounce_03}}
]
properties {
 attenuationSettings=/Game/Sounds/Attenuation/ATT_Fort_GolfCart_Movement_Close_Detail
 bOverrideConcurrency=true
 maxDistance=10000
 soundClassObject=&Fort_Player_Foley
 volumeMultiplier=1.4
 attenuationOverrides {falloffDistance=1000, attenuationShapeExtents {x=500}}
 concurrencyOverrides {resolutionRule=PreventNew}
}
